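# Stereo MetaSound source for a Music Party arcade speaker, replaying an audio bus as spatialized music.
# An Audio Bus Reader takes the bus supplied as input and feeds each channel through one-pole high-pass then low-pass filters with input-driven cutoffs.
# The Set input passes straight to the IsSet output.
# Speaker attenuation, a POI music sound class and a 3D licensed-content submix send apply, with the base submix disabled.
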MetaSoundSource /Hera/Audio/Arcade/MSS_MusicParty_Arcade_Speaker
inputs [
 AudioBus {type=AudioBusAsset}
 HPF {type=Float}
 LPF {type=Float}
 Set {type=Bool}
 UE.Source.OnPlay {type=Trigger}
]
interfaces [UE.OutputFormat.Stereo:1.0, UE.Source:1.1]
outputs [
 IsSet {type=Bool}
 "UE.OutputFormat.Stereo.Audio:0" {type=Audio}
 "UE.OutputFormat.Stereo.Audio:1" {type=Audio}
]
pages [Default {
  input_defaults {AudioBus=null, HPF=0, LPF=20000, Set=null, UE.Source.OnPlay=null}
  nodes [
   AudioBusReader2_0 = "UE.Audio Bus Reader (2)"(Audio Bus=^AudioBus)
   OnePoleHighPassFilter_1 = UE.One-Pole High Pass Filter:Audio(Cutoff Frequency=^HPF, In=@AudioBusReader2_0.Out 1)
   OnePoleLowPassFilter_0 = UE.One-Pole Low Pass Filter:Audio(Cutoff Frequency=^LPF, In=@OnePoleHighPassFilter_1.out)
   OnePoleHighPassFilter_0 = UE.One-Pole High Pass Filter:Audio(Cutoff Frequency=^HPF, In=@AudioBusReader2_0.Out 0)
   OnePoleLowPassFilter_1 = UE.One-Pole Low Pass Filter:Audio(Cutoff Frequency=^LPF, In=@OnePoleHighPassFilter_0.out)
  ]
  links [
   {dst="@IsSet", src="@Set"}
   {dst="@UE.OutputFormat.Stereo.Audio:1", src=OnePoleLowPassFilter_0.out}
   {dst="@UE.OutputFormat.Stereo.Audio:0", src=OnePoleLowPassFilter_1.out}
  ]
 }]
properties {
 assetImportData=&self:AssetImportData
 attenuationSettings=/Game/Sounds/Fort_Audio_Blueprints/MusicParty/MusicPartySpeaker_Att
 bDebug=true
 bEnableBaseSubmix=false
 bLooping=true
 duration=10000
 maxDistance=1400
 numChannels=2
 outputFormat=Stereo
 sampleRate=48000
 soundClassObject=/Game/Sounds/FortSoundClassesAndModes/Fort_Music_POI_PSM
 volume=0.3
 soundSubmixSends [{}, {soundSubmix=/Game/Sounds/SubMixes/Fortnite_LicensedContent_3D}]
}
