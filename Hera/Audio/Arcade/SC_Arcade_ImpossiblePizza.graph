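# Sound cue for the looping "Impossible Pizza" arcade machine music in the Hera audio set.
# A random delay of up to a few seconds precedes a looping wave player.
# It uses the shared arcade cabinet attenuation asset and an arcade machine concurrency group, plays quietly under the POI music sound class, and restarts with seek when virtualized.
# The debug flag is left on.
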SoundCue /Hera/Audio/Arcade/SC_Arcade_ImpossiblePizza
root=Delay_1
nodes [
 {id=Delay_1, children [WavePlayer_0], properties {delayMax=3}}
 {id=WavePlayer_0, properties {bLooping=true, soundWaveAssetPtr=&/MX_Arcade_ImpossiblePizza_LP}}
]
properties {
 attenuationSettings=/Hermes/Props/Commerce/Commerce_ArcadeCabinet_A/Audio/ATT_ArcadeMachines
 bDebug=true
 maxDistance=3000
 soundClassObject=/Game/Sounds/FortSoundClassesAndModes/Fort_Music_POI_PSM
 virtualizationMode=SeekRestart
 volumeMultiplier=0.2
 attenuationOverrides {bAttenuate=false, bSpatialize=false}
 concurrencySet [&/CONC_Hera_ArcadeMachines]
}
